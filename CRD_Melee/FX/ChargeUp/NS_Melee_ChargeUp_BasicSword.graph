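# Niagara melee charge-up effect for the basic sword, timed by System.ChargeDuration.
# During the charge, flares, streaks and a mesh corona follow a two-point socket flow on the user-supplied skeletal mesh, driven by System.DurationCurve.
# Emitters delayed by that duration add the completion flash, mesh burst, spirals, tip flare, GPU sparks and smoke.
# A shared NMS_Melee_SetColor module tints every emitter blue, block-break orange or a custom user color.
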NiagaraSystem /CRD_Melee/FX/ChargeUp/NS_Melee_ChargeUp_BasicSword
dependencies {
 usedDataInterfaces [
  &NiagaraDataInterfaceCurve
  &NiagaraDataInterfaceVector2DCurve
  &NiagaraDataInterfaceVectorCurve
 ]
 usedDynamicInputs [&ClampFloat, &Divide_Float]
 usedModules [
  &self:SystemSpawnScript.SystemScriptSource.SystemScriptGraph.NiagaraNodeAssignment_0.SetVariables_#1
  &SystemState
  &self:SystemSpawnScript.SystemScriptSource.SystemScriptGraph.NiagaraNodeAssignment_1.SetVariables_#2
 ]
}
emitters {
 Animorphic {
  emitterUpdateScript [
   &EmitterState(
    Inactive Response=::NewEnumerator2
    Life Cycle Mode=::NewEnumerator1
    Loop Behavior=::NewEnumerator1
    Loop Duration=$System.ChargeDuration
    UseLoopDelay=true
   )
   &SpawnBurst_Instantaneous(Spawn Count=2, Use Spawn Probability=false, UseLoopCountLimit=false)
  ]
  particleSpawnScript [
   /Niagara/Modules/Spawn/Initialization/InitializeParticle(
    Color=<rgba 0.099 0.6681 1 1>
    Lifetime=$System.ChargeDuration
    Module.Write Color=true
    Module.Write Lifetime=true
    Module.Write Mass=true
    Module.Write Material Random=true
    Module.Write Position=false
    Module.Write Scale=false
    Module.Write SpriteRotation=false
    Module.Write SpriteSize=true
    Sprite Size=&RandomRangeVector2D(Minimum=<175 15>, Maximum=<250 30>, Evaluation Type=::NewEnumerator1)
   )
   /SetVariables(Particles.Color=&LerpLinearColors(StartColor=$Particles.Color, EndColor=<rgba 1 0.4115 0.1302 1>, LerpFactor=$User.Blockbreak)) {enabled=false, isSetParameters=true}
   /CRD_Melee/FX/NMS_Melee_SetColor(
    Block Break:NiagaraFloat=$User.Blockbreak
    Block Break Color:LinearColor=<rgba 1 0.4115 0.1302 1>
    Custom Color:LinearColor=$User.Color
    Custom Color Intensity:NiagaraFloat=$User.CustomColorIntensity
    Non Block Break Color:LinearColor=<rgba 0.099 0.6681 1 1>
    UseCustomColor:NiagaraFloat=$User.UseCustomColor
   )
  ]
  particleUpdateScript [
   &ParticleState(Kill Particles When Lifetime Has Elapsed=false, Let Infinitely Lived Particles Die When Emitter Deactivates=true)
   &/FN_Linear2PointSocketFlow(AnimationTime:NiagaraFloat=$System.DurationCurve, SkeletalMesh:NiagaraDataInterfaceSkeletalMesh=$User.SkeletalMesh)
   &ScaleSpriteSize()
   &ScaleColor(ScaleA=true, ScaleRGB=true)
   &CameraOffset(Camera Offset Amount=10)
   &DynamicMaterialParameters(
    Index 0 Param 1=0.2
    Index 0 Param 2=1
    Index 0 Param 3=1
    Index 0 Param 4=1
    Module.Param0WriteEnabled=true
    Module.Param1WriteEnabled=true
    Module.Param2WriteEnabled=true
    Module.Param3WriteEnabled=true
   )
  ]
  rendererClasses [&NiagaraSpriteRendererProperties]
  renderers [{rendererClass=&NiagaraSpriteRendererProperties}]
 }
 Animorphic001 {
  emitterUpdateScript [
   &EmitterState(
    Inactive Response=::NewEnumerator2
    Life Cycle Mode=::NewEnumerator1
    Loop Behavior=::NewEnumerator1
    Loop Delay=$System.ChargeDuration
    Loop Duration Mode=::NewEnumerator1
    UseLoopDelay=true
   )
   &SpawnBurst_Instantaneous(Use Spawn Probability=false, UseLoopCountLimit=false)
  ]
  particleSpawnScript [
   &InitializeParticle(
    Color=<rgba 0.4136 1.9216 3 1>
    Lifetime=0.4
    Module.Write Color=true
    Module.Write Lifetime=true
    Module.Write Mass=true
    Module.Write Material Random=true
    Module.Write Position=false
    Module.Write Scale=false
    Module.Write SpriteRotation=false
    Module.Write SpriteSize=true
    Sprite Size=&Multiply_Vector2D_ByFloat(Float=1.5, Vector2D=&RandomRangeVector2D(Minimum=<175 15>, Maximum=<250 30>))
   )
   /SetVariables(Particles.Color=&LerpLinearColors(StartColor=$Particles.Color, EndColor=<rgba 3 1.2537 0.3702 1>, LerpFactor=$User.Blockbreak)) {enabled=false, isSetParameters=true}
   &NMS_Melee_SetColor(
    Block Break:NiagaraFloat=$User.Blockbreak
    Block Break Color:LinearColor=<rgba 3 1.2537 0.3702 1>
    Custom Color:LinearColor=$User.Color
    Custom Color Intensity:NiagaraFloat=$User.CustomColorIntensity
    Custom Color Multiplier=3
    Non Block Break Color:LinearColor=<rgba 0.4136 1.9216 3 1>
    UseCustomColor:NiagaraFloat=$User.UseCustomColor
   )
  ]
  particleUpdateScript [
   &ParticleState()
   &SkeletalMeshLocation(
    Bone / Socket Index=1
    MaskSpawnGroup=false
    Mesh Sampling Type=::NewEnumerator5
    Skeletal Mesh=$User.SkeletalMesh
    Socket Sampling Mode=::NewEnumerator5
   )
   &ScaleSpriteSize(Uniform Curve Sprite Scale=DataInterface() {propertyValues {Curve {keys [
        {arriveTangent=-0.8089, interpMode=RCIM_Cubic, leaveTangent=-0.8089, tangentMode=RCTM_User, time=0, value=6}
        {interpMode=RCIM_Cubic, tangentMode=RCTM_User, time=1, value=0}
       ]}}})
   &ScaleColor(ScaleA=true, ScaleRGB=true)
   &CameraOffset(Camera Offset Amount=10)
   &DynamicMaterialParameters(
    Index 0 Param 1=0.2
    Index 0 Param 2=1
    Index 0 Param 3=1
    Index 0 Param 4=1
    Module.Param0WriteEnabled=true
    Module.Param1WriteEnabled=true
    Module.Param2WriteEnabled=true
    Module.Param3WriteEnabled=true
   )
  ]
  rendererClasses [&NiagaraSpriteRendererProperties]
  renderers [{rendererClass=&NiagaraSpriteRendererProperties}]
 }
 Charge_Flash {
  emitterUpdateScript [
   &EmitterState(
    Inactive Response=::NewEnumerator2
    Life Cycle Mode=::NewEnumerator1
    Loop Behavior=::NewEnumerator1
    Loop Delay=$System.ChargeDuration
    Loop Duration Mode=::NewEnumerator1
    UseLoopDelay=true
   )
   &SpawnBurst_Instantaneous(Use Spawn Probability=false, UseLoopCountLimit=false)
  ]
  particleSpawnScript [
   /Niagara/Modules/Spawn/Initialization/V2/InitializeParticle(
    Color=<rgba 0.7633 3.5252 5 0.3>
    Lifetime=0.3
    Sprite Size Mode=::NewEnumerator3
    Uniform Sprite Size=260
    UsePositionOffset=false
    Write Material Random=true
   )
   /SetVariables(Particles.Color=&LerpLinearColors(StartColor=$Particles.Color, EndColor=<rgba 3 1.2537 0.3702 1>, LerpFactor=$User.Blockbreak)) {enabled=false, isSetParameters=true}
   &NMS_Melee_SetColor(
    Block Break:NiagaraFloat=$User.Blockbreak
    Block Break Color:LinearColor=<rgba 3 1.2537 0.3702 1>
    Custom Color:LinearColor=$User.Color
    Custom Color Intensity:NiagaraFloat=$User.CustomColorIntensity
    Custom Color Multiplier=3
    Custom Color Opacity=0.75
    Non Block Break Color:LinearColor=<rgba 0.7633 3.5252 5 0.3>
    UseCustomColor:NiagaraFloat=$User.UseCustomColor
   )
  ]
  particleUpdateScript [
   &ParticleState()
   &ScaleColor(
    Scale Alpha=&FloatFromCurve(FloatCurve=DataInterface() {propertyValues {Curve {keys [{arriveTangent=-23.5419, interpMode=RCIM_Cubic, time=0.2, value=8}, <at 1 0 RCIM_Cubic>]}}})
    ScaleA=true
    ScaleRGB=true
   )
   &ScaleSpriteSize(Uniform Curve Sprite Scale=DataInterface() {propertyValues {Curve {keys [{interpMode=RCIM_Cubic, tangentMode=RCTM_User, time=0, value=2}, {interpMode=RCIM_Cubic, tangentMode=RCTM_User, time=1, value=0}]}}})
   &CameraOffset(Camera Offset Amount=35)
   &DynamicMaterialParameters(
    Index 0 Param 1=0.2
    Index 0 Param 3=1
    Index 0 Param 4=1
    Module.Param0WriteEnabled=true
    Module.Param1WriteEnabled=true
    Module.Param2WriteEnabled=true
    Module.Param3WriteEnabled=true
   )
   &SkeletalMeshLocation(
    Bone / Socket Index=1
    MaskSpawnGroup=false
    Mesh Sampling Type=::NewEnumerator5
    Skeletal Mesh=$User.SkeletalMesh
    Socket Sampling Mode=::NewEnumerator5
   )
  ]
  rendererClasses [&NiagaraSpriteRendererProperties]
  renderers [{rendererClass=&NiagaraSpriteRendererProperties}]
 }
 Curled_Sparks {
  simTarget=GPUComputeSim
  emitterUpdateScript [
   &EmitterState(Delay First Loop Only=true, Life Cycle Mode=::NewEnumerator1, Loop Delay=$System.ChargeDuration, Loop Duration=1, UseLoopDelay=true)
   &SpawnRate(SpawnRate=60, Use Spawn Probability=false)
  ]
  particleSpawnScript [
   &InitializeParticle(
    Lifetime Max=0.8
    Lifetime Min=0.15
    Lifetime Mode=::NewEnumerator1
    Sprite Size Max=<1.5 5>
    Sprite Size Min=<1 3>
    Sprite Size Mode=::NewEnumerator2
    UsePositionOffset=false
    Write Material Random=true
   )
   /SetVariables(Particles.Color=&LerpLinearColors(StartColor=<rgba 0.1527 0.705 1 0.3>, EndColor=<rgba 1 0.3704 0.1302 0.3>, LerpFactor=$User.Blockbreak)) {enabled=false, isSetParameters=true}
   &NMS_Melee_SetColor(
    Block Break:NiagaraFloat=$User.Blockbreak
    Block Break Color:LinearColor=<rgba 1 0.3704 0.1302 0.3>
    Custom Color:LinearColor=$User.Color
    Custom Color Intensity:NiagaraFloat=$User.CustomColorIntensity
    Custom Color Opacity=0.3
    Non Block Break Color:LinearColor=<rgba 0.1527 0.705 1 0.3>
    UseCustomColor:NiagaraFloat=$User.UseCustomColor
   )
   &AddVelocity(
    Cone Angle=0
    Cone Axis=<0 0 1>
    Module.UseConeFalloff=false
    Module.UseVelDistribution=false
    Override Seed=false
    Velocity Mode=::NewEnumerator2
    Velocity Speed=&RandomRangeFloat(Minimum=60, Maximum=200)
   )
   &ApplyInitialForces()
   &SphereLocation(
    MaskSpawnGroup=false
    Non Uniform Scale=<1 1 5>
    Offset=<0 0 50>
    Override Seed=false
    Sphere Radius=9
    Surface Only=false
    UseNonUniformScale=true
    UseOffset=true
    UseSphereOrientation=false
   )
  ]
  particleUpdateScript [
   &ParticleState()
   &ScaleColor(
    Scale Alpha=&FloatFromCurve(FloatCurve=DataInterface() {propertyValues {Curve {keys [
         {interpMode=RCIM_Cubic, tangentMode=RCTM_User, time=0, value=0}
         {interpMode=RCIM_Cubic, tangentMode=RCTM_User, time=0.2, value=1.125}
         {tangentMode=RCTM_User, time=1, value=0}
        ]}}})
    Scale RGB=<10 10 10>
    ScaleA=true
    ScaleRGB=true
   )
   &ScaleSpriteSize(
    Scale Factor=&Vector2DFromCurve(
     Vector2Curve=DataInterface() {propertyValues {XCurve {keys [{arriveTangent=-1.8979, time=0, value=1}, <at 1 0.1>]}, YCurve {keys [<at 0 1.5>,<at 1 0.5>]}}}
     ScaleCurve=<1 1>
    )
    Scale Sprite Size Mode=::NewEnumerator2
   )
   &CameraOffset(Camera Offset Amount=35)
   &DynamicMaterialParameters(
    Index 0 Param 1=0.2
    Index 0 Param 3=1
    Index 0 Param 4=1
    Module.Param0WriteEnabled=true
    Module.Param1WriteEnabled=true
    Module.Param2WriteEnabled=true
    Module.Param3WriteEnabled=true
   )
   &Drag(Use Linear Drag=true, Use Rotational Drag=false)
   &CurlNoiseForce(Noise Strength=700, Pan Noise=true, Pan Noise Field=<0.5 0.1 0.3>)
   &SolveForcesAndVelocity(Clamp Velocity=false, Limit Acceleration=false)
  ]
  rendererClasses [&NiagaraSpriteRendererProperties]
  renderers [{rendererClass=&NiagaraSpriteRendererProperties}]
 }
 Directional_sparks {
  simTarget=GPUComputeSim
  emitterUpdateScript [
   &EmitterState(Delay First Loop Only=true, Life Cycle Mode=::NewEnumerator1, Loop Delay=$System.ChargeDuration, Loop Duration=1, UseLoopDelay=true)
   &SpawnRate(SpawnRate=15, Use Spawn Probability=false)
  ]
  particleSpawnScript [
   &InitializeParticle(
    Color=<rgba 0.1527 0.705 1 0.3>
    Lifetime Max=0.35
    Lifetime Min=0.18
    Lifetime Mode=::NewEnumerator1
    Sprite Size Max=<2 9>
    Sprite Size Min=<1 5>
    Sprite Size Mode=::NewEnumerator2
    UsePositionOffset=false
    Write Material Random=true
   )
   /SetVariables(Particles.Color=&LerpLinearColors(StartColor=$Particles.Color, EndColor=<rgba 1 0.3704 0.1302 0.3>, LerpFactor=$User.Blockbreak)) {enabled=false, isSetParameters=true}
   &NMS_Melee_SetColor(
    Block Break:NiagaraFloat=$User.Blockbreak
    Block Break Color:LinearColor=<rgba 1 0.3704 0.1302 0.3>
    Custom Color:LinearColor=$User.Color
    Custom Color Intensity:NiagaraFloat=$User.CustomColorIntensity
    Custom Color Opacity=0.3
    Non Block Break Color:LinearColor=<rgba 0.1527 0.705 1 0.3>
    UseCustomColor:NiagaraFloat=$User.UseCustomColor
   )
   &AddVelocity(
    Cone Angle=0
    Cone Axis=<0 0 1>
    Module.UseConeFalloff=false
    Module.UseVelDistribution=false
    Override Seed=false
    Velocity Mode=::NewEnumerator2
    Velocity Speed=&RandomRangeFloat(Minimum=350, Maximum=600)
   )
   &ApplyInitialForces()
   &SphereLocation(
    MaskSpawnGroup=false
    Non Uniform Scale=<1 1 5>
    Offset=<0 0 40>
    Override Seed=false
    Sphere Radius=9
    Surface Only=false
    UseNonUniformScale=true
    UseOffset=true
    UseSphereOrientation=false
   )
  ]
  particleUpdateScript [
   &ParticleState()
   &ScaleColor(
    Scale Alpha=&FloatFromCurve(FloatCurve=DataInterface() {propertyValues {Curve {keys [
         {interpMode=RCIM_Cubic, tangentMode=RCTM_User, time=0, value=0}
         {interpMode=RCIM_Cubic, tangentMode=RCTM_User, time=0.2, value=1.125}
         {tangentMode=RCTM_User, time=1, value=0}
        ]}}})
    Scale RGB=<10 10 10>
    ScaleA=true
    ScaleRGB=true
   )
   &ScaleSpriteSize(
    Scale Factor=&Vector2DFromCurve(
     Vector2Curve=DataInterface() {propertyValues {XCurve {keys [{arriveTangent=-1.8979, time=0, value=1}, <at 1 0.1>]}, YCurve {keys [<at 0 2>,<at 1 0.5>]}}}
     ScaleCurve=<1 1>
    )
    Scale Sprite Size Mode=::NewEnumerator2
   )
   &CameraOffset(Camera Offset Amount=35)
   &DynamicMaterialParameters(
    Index 0 Param 1=0.2
    Index 0 Param 3=1
    Index 0 Param 4=1
    Module.Param0WriteEnabled=true
    Module.Param1WriteEnabled=true
    Module.Param2WriteEnabled=true
    Module.Param3WriteEnabled=true
   )
   &Drag(Use Linear Drag=true, Use Rotational Drag=false)
   &CurlNoiseForce(Noise Strength=50, Pan Noise=false)
   &SolveForcesAndVelocity(Clamp Velocity=false, Limit Acceleration=false)
  ]
  rendererClasses [&NiagaraSpriteRendererProperties]
  renderers [{rendererClass=&NiagaraSpriteRendererProperties}]
 }
 Flares {
  emitterUpdateScript [
   &EmitterState(
    Inactive Response=::NewEnumerator2
    Life Cycle Mode=::NewEnumerator1
    Loop Behavior=::NewEnumerator1
    Loop Duration=$System.ChargeDuration
    UseLoopDelay=false
   )
   &SpawnBurst_Instantaneous(Spawn Count=4, Use Spawn Probability=false, UseLoopCountLimit=false)
  ]
  particleSpawnScript [
   &InitializeParticle(
    Color=<rgba 0.3476 2.3473 4 0.3>
    Lifetime=&Add_Float(A=$System.ChargeDuration, B=0.1)
    Sprite Rotation Mode=::NewEnumerator2
    Sprite Size Mode=::NewEnumerator1
    Sprite UV Mode=::NewEnumerator3
    Uniform Sprite Size Max=120
    Uniform Sprite Size Min=80
    UsePositionOffset=false
    Write Material Random=true
   )
   /SetVariables(Particles.Color=&LerpLinearColors(StartColor=$Particles.Color, EndColor=<rgba 4 1.9919 1.1012 1>, LerpFactor=$User.Blockbreak)) {enabled=false, isSetParameters=true, module=SetVariables_#3}
   &NMS_Melee_SetColor(
    Block Break:NiagaraFloat=$User.Blockbreak
    Block Break Color:LinearColor=<rgba 4 1.9919 1.1012 1>
    Custom Color:LinearColor=$User.Color
    Custom Color Intensity:NiagaraFloat=$User.CustomColorIntensity
    Custom Color Multiplier=4
    Custom Color Opacity=0.5
    Non Block Break Color:LinearColor=<rgba 0.3476 2.3473 4 0.3>
    UseCustomColor:NiagaraFloat=$User.UseCustomColor
   )
   &DynamicMaterialParameters(
    Index 0 Param 1=&RandomRangeFloat(Minimum=0.1, Maximum=0.3)
    Index 0 Param 2=&RandomRangeFloat(Minimum=0.8, Maximum=1.3)
    Index 0 Param 3=1
    Index 0 Param 4=1
    Module.Param0WriteEnabled=true
    Module.Param1WriteEnabled=true
    Module.Param2WriteEnabled=true
    Module.Param3WriteEnabled=true
   )
  ]
  particleUpdateScript [
   &ParticleState()
   &FN_Linear2PointSocketFlow(AnimationTime:NiagaraFloat=$System.DurationCurve, SkeletalMesh:NiagaraDataInterfaceSkeletalMesh=$User.SkeletalMesh)
   &SpriteRotationRate(Rotation Rate=&RandomRangeFloat(Minimum=-60, Maximum=60))
   &ScaleSpriteSize()
   &CameraOffset(Camera Offset Amount=35)
   &ScaleColor(
    Scale Alpha=&FloatFromCurve(FloatCurve=DataInterface() {propertyValues {Curve {keys [
         {interpMode=RCIM_Cubic, tangentMode=RCTM_User, time=0, value=0}
         {arriveTangent=-0.1667, interpMode=RCIM_Cubic, leaveTangent=-0.1667, tangentMode=RCTM_User, time=0.1, value=1}
         {arriveTangent=0.2218, interpMode=RCIM_Cubic, leaveTangent=0.2218, tangentMode=RCTM_User, time=0.9, value=1}
         {arriveTangent=-0.1667, interpMode=RCIM_Cubic, leaveTangent=-0.1667, tangentMode=RCTM_User, time=1, value=0}
        ]}}})
    ScaleA=true
    ScaleRGB=true
   )
  ]
  rendererClasses [&NiagaraSpriteRendererProperties]
  renderers [{rendererClass=&NiagaraSpriteRendererProperties}]
 }
 Impact_Flare_and_Corona {
  emitterUpdateScript [
   &EmitterState(Life Cycle Mode=::NewEnumerator1, Loop Behavior=::NewEnumerator1, Loop Duration=$System.ChargeDuration, UseLoopDelay=false)
   &SpawnBurst_Instantaneous(Spawn Count=1, Use Spawn Probability=false)
   &SpawnBurst_Instantaneous(Spawn Time=0.1, Use Spawn Probability=false, UseLoopCountLimit=false) {module=SpawnBurst_Instantaneous001}
  ]
  particleSpawnScript [
   /Niagara/Modules/Spawn/Initialization/InitializeParticle(
    Color=<rgba 0 2.229 4 1>
    Lifetime=$System.ChargeDuration
    Mesh Scale=<0.25 0.25 0.25>
    Module.Write Color=true
    Module.Write Lifetime=true
    Module.Write Mass=true
    Module.Write Material Random=true
    Module.Write Position=false
    Module.Write Scale=true
    Module.Write SpriteRotation=false
    Module.Write SpriteSize=true
    Position=&SimulationPosition()
    Sprite Size=<200 150>
   )
   /SetVariables(Particles.Color=&LerpLinearColors(StartColor=$Particles.Color, EndColor=<rgba 4 1.9728 0.8739 1>, LerpFactor=$User.Blockbreak)) {enabled=false, isSetParameters=true}
   &NMS_Melee_SetColor(
    Block Break:NiagaraFloat=$User.Blockbreak
    Block Break Color:LinearColor=<rgba 4 1.9728 0.8739 1>
    Custom Color:LinearColor=$User.Color
    Custom Color Intensity:NiagaraFloat=$User.CustomColorIntensity
    Custom Color Multiplier=4
    Non Block Break Color:LinearColor=<rgba 0 2.229 4 1>
    UseCustomColor:NiagaraFloat=$User.UseCustomColor
   )
   &CameraOffset(Camera Offset Amount=8)
  ]
  particleUpdateScript [
   &ParticleState(Kill Particles When Lifetime Has Elapsed=false, Let Infinitely Lived Particles Die When Emitter Deactivates=true)
   &FN_Linear2PointSocketFlow(AnimationTime:NiagaraFloat=$System.DurationCurve, SkeletalMesh:NiagaraDataInterfaceSkeletalMesh=$User.SkeletalMesh)
   &DynamicMaterialParameters(
    Index 0 Param 1=&RandomRangeFloat(Minimum=2, Maximum=20)
    Index 0 Param 2=&RandomRangeFloat(Minimum=0.1, Maximum=1)
    Index 0 Param 3=1
    Index 0 Param 4=&FloatFromCurve(FloatCurve=DataInterface() {propertyValues {Curve {keys [
         <at 0 0 RCIM_Cubic>
         {arriveTangent=9.5022, interpMode=RCIM_Cubic, leaveTangent=9.5022, tangentMode=RCTM_User, time=0.4, value=1}
         <at 1 -0.6 RCIM_Cubic>
        ]}}})
    Module.Param0WriteEnabled=true
    Module.Param1WriteEnabled=true
    Module.Param2WriteEnabled=true
    Module.Param3WriteEnabled=true
   )
   &ScaleMeshSize(Scale Factor=&VectorFromFloat(Value=&FloatFromCurve(FloatCurve=DataInterface() {propertyValues {Curve {keys [
          {arriveTangent=-8.6268, interpMode=RCIM_Cubic, leaveTangent=-8.6268, tangentMode=RCTM_User, time=0, value=1.5}
          {arriveTangent=1, leaveTangent=1, time=0.4, value=0}
         ]}}})))
   &UpdateMeshOrientation(Rotation Rate=0.5, Rotation Vector=<1 0 0>)
  ]
  rendererClasses [&NiagaraMeshRendererProperties]
  renderers [{rendererClass=&NiagaraMeshRendererProperties, rendererIndex=0}]
 }
 Mesh {
  emitterUpdateScript [
   &EmitterState(
    Inactive Response=::NewEnumerator2
    Life Cycle Mode=::NewEnumerator1
    Loop Behavior=::NewEnumerator1
    Loop Delay=$System.ChargeDuration
    Loop Duration Mode=::NewEnumerator1
    UseLoopDelay=true
   )
   &SpawnBurst_Instantaneous(Use Spawn Probability=false, UseLoopCountLimit=false)
  ]
  particleSpawnScript [
   /Niagara/Modules/Spawn/Initialization/V2/InitializeParticle(
    Color=<rgba 0 1.1057 10 0.3>
    Lifetime=0.3
    Mesh Scale Mode=::NewEnumerator3
    Sprite Size Mode=::NewEnumerator3
    Uniform Sprite Size=220
    UsePositionOffset=false
    Write Material Random=true
   )
   /SetVariables(Particles.Color=&LerpLinearColors(StartColor=$Particles.Color, EndColor=<rgba 5 1.852 0.651 0.3>, LerpFactor=$User.Blockbreak)) {enabled=false, isSetParameters=true}
   &NMS_Melee_SetColor(
    Block Break:NiagaraFloat=$User.Blockbreak
    Block Break Color:LinearColor=<rgba 5 1.852 0.651 0.3>
    Custom Color:LinearColor=$User.Color
    Custom Color Intensity:NiagaraFloat=$User.CustomColorIntensity
    Custom Color Multiplier=7.5
    Custom Color Opacity=0.3
    Non Block Break Color:LinearColor=<rgba 0 1.1057 10 0.3>
    UseCustomColor:NiagaraFloat=$User.UseCustomColor
   )
   &DynamicMaterialParameters(
    Index 0 Param 1=3
    Index 0 Param 2=1
    Index 0 Param 3=10
    Index 0 Param 4=10
    Module.Param0WriteEnabled=true
    Module.Param1WriteEnabled=true
    Module.Param2WriteEnabled=true
    Module.Param3WriteEnabled=true
   )
   &DynamicMaterialParameters(
    Index 1 Param 1=1
    Module.Param0WriteEnabled1=true
    Module.Param1WriteEnabled1=true
    Module.Param2WriteEnabled1=true
    Module.Param3WriteEnabled1=true
    Write Parameter Index 0=false
    Write Parameter Index 1=true
   ) {enabled=false, module=DynamicMaterialParameters001}
  ]
  particleUpdateScript [
   &ParticleState(Kill Particles When Lifetime Has Elapsed=false, Let Infinitely Lived Particles Die When Emitter Deactivates=true)
   &ScaleColor(
    Scale RGB=&VectorFromFloat(Value=&FloatFromCurve(
      FloatCurve=DataInterface() {propertyValues {Curve {keys [
          {arriveTangent=-450.6744, interpMode=RCIM_Cubic, leaveTangent=-450.6744, tangentMode=RCTM_User, time=0, value=100}
          {interpMode=RCIM_Cubic, tangentMode=RCTM_User, time=0.5, value=5}
         ]}}}
      CurveIndex=$Particles.Age
     ))
    ScaleA=true
    ScaleRGB=true
   )
   &CameraOffset(Camera Offset Amount=35) {enabled=false}
   &SkeletalMeshLocation(
    Bone / Socket Index=1
    MaskSpawnGroup=false
    Mesh Sampling Type=::NewEnumerator5
    Skeletal Mesh=$User.SkeletalMesh
    Socket Sampling Mode=::NewEnumerator5
   )
  ]
  rendererClasses [&NiagaraMeshRendererProperties]
  renderers [{rendererClass=&NiagaraMeshRendererProperties, rendererIndex=0}]
 }
 Smoke_Wide_Large {
  simTarget=GPUComputeSim
  emitterUpdateScript [
   &EmitterState(Delay First Loop Only=true, Life Cycle Mode=::NewEnumerator1, Loop Delay=$System.ChargeDuration, Loop Duration=1, UseLoopDelay=true)
   &SpawnRate(SpawnRate=40, Use Spawn Probability=false)
  ]
  particleSpawnScript [
   &InitializeParticle(
    Lifetime Max=0.6
    Lifetime Min=0.35
    Lifetime Mode=::NewEnumerator1
    Sprite Rotation Mode=::NewEnumerator2
    Sprite Size Mode=::NewEnumerator1
    Uniform Sprite Size Max=50
    Uniform Sprite Size Min=30
    UsePositionOffset=false
    Write Material Random=true
   ) {module=InitializeParticle001}
   /SetVariables(Particles.Color=&LerpLinearColors(StartColor=<rgba 0.2008 0.5834 0.9479 1>, EndColor=<rgba 1 0.5525 0.2448 1>, LerpFactor=$User.Blockbreak)) {enabled=false, isSetParameters=true}
   &NMS_Melee_SetColor(
    Block Break:NiagaraFloat=$User.Blockbreak
    Block Break Color:LinearColor=<rgba 1 0.5525 0.2448 1>
    Custom Color:LinearColor=$User.Color
    Custom Color Intensity:NiagaraFloat=$User.CustomColorIntensity
    Non Block Break Color:LinearColor=<rgba 0.2008 0.5834 0.9479 1>
    UseCustomColor:NiagaraFloat=$User.UseCustomColor
   )
   /SetVariables(Particles.InitSpriteRotRate=&UniformRangedFloat(Minimum=-3, Maximum=3)) {isSetParameters=true}
   &SkeletalMeshLocation(
    Bone / Socket Index=&SpawnGroup()
    Inherit Velocity=false
    Interpolate Mesh Sampling=false
    MaskSpawnGroup=false
    Mesh Sampling Type=::NewEnumerator5
    Skeletal Mesh=$User.SkeletalMesh
    Socket Sampling Mode=::NewEnumerator5
   )
   &AddVelocityInCone(
    Cone Angle=15
    Cone Axis=<0 0 1>
    Module.Use Velocity Falloff On Cone Axis=false
    Velocity Strength=&RandomRangeFloat(Minimum=500, Maximum=750)
   )
   &SphereLocation(
    MaskSpawnGroup=false
    Non Uniform Scale=<1 1 5>
    Override Seed=false
    Sphere Radius=8
    Surface Only=false
    UseNonUniformScale=true
    UseOffset=true
    UseSphereOrientation=false
   ) {module=SphereLocation001}
  ]
  particleUpdateScript [
   &ParticleState()
   &SpriteRotationRate(Rotation Rate=&Multiply_Float(
     A=$Particles.InitSpriteRotRate
     B=&FloatFromCurve(
      FloatCurve=DataInterface() {propertyValues {Curve {keys [
          {arriveTangent=-1.9214, interpMode=RCIM_Cubic, leaveTangent=-1.9214, tangentMode=RCTM_User, time=-0.0046, value=1.0113}
          {arriveTangent=-0.2496, interpMode=RCIM_Cubic, leaveTangent=-0.2496, tangentMode=RCTM_User, time=0.9896, value=0.16}
         ]}}}
      CurveIndex=$Emitter.Age
      Scale Curve=150
     )
    ))
   &ScaleColor(
    Scale Alpha=&FloatFromCurve(
     FloatCurve=DataInterface() {propertyValues {Curve {keys [
         {interpMode=RCIM_Cubic, tangentMode=RCTM_User, time=0, value=0}
         {interpMode=RCIM_Cubic, tangentMode=RCTM_User, time=0.4, value=0.95}
         {tangentMode=RCTM_User, time=1, value=0}
        ]}}}
     Scale Curve=0.15
    )
    ScaleA=true
    ScaleRGB=true
   )
   &Drag(Drag=6, Use Linear Drag=true, Use Rotational Drag=false)
   &GravityForce(Gravity=&Lerp_Vector(A=<0 0 400>, B=<0 0 -100>, Alpha=$Particles.MaterialRandom))
   &SolveForcesAndVelocity(Clamp Velocity=false, Limit Acceleration=false)
   &ScaleSpriteSize(Scale Factor=&Vector2DFromFloat(Value=&FloatFromCurve(FloatCurve=DataInterface() {propertyValues {Curve {keys [{arriveTangent=3.172, interpMode=RCIM_Cubic, time=0, value=0.2}, {interpMode=RCIM_Cubic, leaveTangent=1.6352, time=0.2, value=1}]}}})))
   &DynamicMaterialParameters(
    Index 0 Param 1=&Add_Float(
     A=$Particles.RandomFloat
     B=&FloatFromCurve(FloatCurve=DataInterface() {propertyValues {Curve {keys [
          {arriveTangent=0.9088, interpMode=RCIM_Cubic, leaveTangent=0.9088, tangentMode=RCTM_User, time=0, value=1}
          {arriveTangent=0.5473, interpMode=RCIM_Cubic, leaveTangent=0.5473, tangentMode=RCTM_User, time=1, value=2}
         ]}}})
    )
    Index 0 Param 2=&FloatFromCurve(FloatCurve=DataInterface() {propertyValues {Curve {keys [
         {arriveTangent=0.0453, interpMode=RCIM_Cubic, tangentMode=RCTM_Break, time=0, value=-1}
         {arriveTangent=0.0604, interpMode=RCIM_Cubic, leaveTangent=0.0604, tangentMode=RCTM_User, time=1, value=-1}
        ]}}})
    Index 1 Param 1=&FloatFromCurve(
     FloatCurve=DataInterface() {propertyValues {Curve {keys [<at 0 0>,<at 0.25 1>]}}}
     Scale Curve=&RandomRangeFloat(Minimum=2, Maximum=1)
    )
    Index 1 Param 2=&FloatFromCurve(
     FloatCurve=DataInterface() {propertyValues {Curve {keys [
         {interpMode=RCIM_Cubic, leaveTangent=3.7422, tangentMode=RCTM_Break, time=0, value=0.2}
         {interpMode=RCIM_Cubic, leaveTangent=0.9066, time=1, value=2}
        ]}}}
     Scale Curve=&RandomRangeFloat(Minimum=0.5, Maximum=1)
    )
    Index 1 Param 3=&FloatFromCurve(FloatCurve=DataInterface() {propertyValues {Curve {keys [
         {interpMode=RCIM_Cubic, leaveTangent=2.8445, tangentMode=RCTM_Break, time=0, value=0.2}
         {interpMode=RCIM_Cubic, leaveTangent=63.2978, time=0.25, value=1}
        ]}}})
    Index 1 Param 4=&FloatFromCurve(
     FloatCurve=DataInterface() {propertyValues {Curve {keys [<at 0 1>,<at 1 0.8>]}}}
     Scale Curve=&RandomRangeFloat(Minimum=0.5, Maximum=1)
    )
    Index 2 Param 1=&FloatFromCurve(FloatCurve=DataInterface() {propertyValues {Curve {keys [{interpMode=RCIM_Cubic, leaveTangent=3.2309, tangentMode=RCTM_Break, time=0.1, value=0}, <at 1 1>]}}})
    Index 2 Param 2=&FloatFromCurve(
     FloatCurve=DataInterface() {propertyValues {Curve {keys [<at 0 0.2 RCIM_Cubic>,<at 1 0.8 RCIM_Cubic>]}}}
     Scale Curve=&RandomRangeFloat(Minimum=0.2, Maximum=1.2)
    )
    Index 2 Param 3=&FloatFromCurve(FloatCurve=DataInterface() {propertyValues {Curve {keys [
         {interpMode=RCIM_Cubic, tangentMode=RCTM_Break, time=0, value=1}
         {interpMode=RCIM_Cubic, leaveTangent=11.1233, tangentMode=RCTM_Break, time=1, value=2}
        ]}}})
    Index 2 Param 4=&FloatFromCurve(
     FloatCurve=DataInterface() {propertyValues {Curve {keys [<at 0 1 RCIM_Cubic>, {interpMode=RCIM_Cubic, leaveTangent=-1.2956, tangentMode=RCTM_Break, time=1, value=0.75}]}}}
     Scale Curve=0.8
    )
    Module.Param0WriteEnabled=true
    Module.Param0WriteEnabled1=true
    Module.Param0WriteEnabled2=true
    Module.Param1WriteEnabled=false
    Module.Param1WriteEnabled1=true
    Module.Param1WriteEnabled2=true
    Module.Param2WriteEnabled=true
    Module.Param2WriteEnabled1=true
    Module.Param2WriteEnabled2=true
    Module.Param3WriteEnabled=true
    Module.Param3WriteEnabled1=true
    Module.Param3WriteEnabled2=true
    Write Parameter Index 1=true
    Write Parameter Index 2=true
   )
  ]
  rendererClasses [&NiagaraSpriteRendererProperties]
  renderers [{rendererClass=&NiagaraSpriteRendererProperties}]
 }
 Spirals {
  emitterUpdateScript [
   &EmitterState(Delay First Loop Only=true, Life Cycle Mode=::NewEnumerator1, Loop Delay=$System.ChargeDuration, Loop Duration=1, UseLoopDelay=true)
   &SpawnRate(SpawnRate=5, Use Spawn Probability=false)
   &SpawnRate(SpawnGroup=1, SpawnRate=5, Use Spawn Probability=false) {module=SpawnRate001}
  ]
  particleSpawnScript [
   &InitializeParticle(
    Lifetime Max=0.7
    Lifetime Min=0.6
    Lifetime Mode=::NewEnumerator1
    Mesh Scale=&Multiply_VectorByFloat(Vector=<0.3 0.3 1>, Float=&RandomRangeFloat(Minimum=0.5, Maximum=0.7))
    Mesh Scale Mode=::NewEnumerator4
    UsePositionOffset=false
    Write Material Random=true
   )
   /SetVariables(Particles.Color=&LerpLinearColors(StartColor=<rgba 0.22 1.0336 2 1>, EndColor=<rgba 1.5 0.9113 0.5275 1>, LerpFactor=$User.Blockbreak)) {enabled=false, isSetParameters=true}
   &NMS_Melee_SetColor(
    Block Break:NiagaraFloat=$User.Blockbreak
    Block Break Color:LinearColor=<rgba 1.5 0.9113 0.5275 1>
    Custom Color:LinearColor=$User.Color
    Custom Color Intensity:NiagaraFloat=$User.CustomColorIntensity
    Custom Color Multiplier=1.5
    Non Block Break Color:LinearColor=<rgba 0.22 1.0336 2 1>
    UseCustomColor:NiagaraFloat=$User.UseCustomColor
   )
   /SetVariables(Particles.Scale=&MaskVectorBySpawnGroup(
     Masked Vector=&RandomRangeVector(Minimum=<0.6 0.6 0.1>, Maximum=<0.8 0.8 0.3>)
     Passthrough Vector=$Particles.Scale
     Spawn Group Mask=1
    )) {isSetParameters=true}
   &InitialMeshOrientation(
    Mesh Orientation Mode=::NewEnumerator7
    Rotation=&MakeVector(X=0, Y=0, Z=&RandomRangeFloat(Minimum=0, Maximum=1))
    Use Rotation Vector=true
   ) {module=InitialMeshOrientation001}
   &BoxLocation(
    Box Offset=&MaskVectorBySpawnGroup(Masked Vector=<0 0 50>, Passthrough Vector=<0 0 0>, Spawn Group Mask=1)
    Box Size=&MaskVectorBySpawnGroup(Masked Vector=<0 0 70>, Passthrough Vector=<0 0 0>, Spawn Group Mask=1)
    MaskSpawnGroup=false
    Module.Override Seed=false
    Rotate Box=false
    Surface Only=false
   )
   &AddVelocity(Velocity=&MaskVectorBySpawnGroup(
     Masked Vector=&RandomRangeVector(Minimum=<0 0 10>, Maximum=<0 0 50>)
     Passthrough Vector=&RandomRangeVector(Minimum=<0 0 50>, Maximum=<0 0 150>)
     Spawn Group Mask=1
    ))
   &AddRotationalVelocity(
    Coordinate Space:ENiagaraMeshTransforms=::NewEnumerator2
    Rotation Rate=&RandomRangeVector(Minimum=<0 0 0.2>, Maximum=<0 0 -2>)
   )
  ]
  particleUpdateScript [
   &ParticleState()
   &Drag(Use Linear Drag=true, Use Rotational Drag=false)
   &SolveForcesAndVelocity(Clamp Velocity=false, Limit Acceleration=false, Speed Limit=1200) {module=SolveForcesAndVelocity001}
   &SolveRotationalForcesAndVelocity()
   &ScaleColor(
    Scale Alpha=&FloatFromCurve(FloatCurve=DataInterface() {propertyValues {Curve {keys [
         {interpMode=RCIM_Cubic, tangentMode=RCTM_User, time=0, value=0}
         {interpMode=RCIM_Cubic, tangentMode=RCTM_User, time=0.5, value=0.5}
         {tangentMode=RCTM_User, time=1, value=0}
        ]}}})
    Scale RGB=&VectorFromFloat(Value=1)
    ScaleA=true
    ScaleRGB=true
   )
   &ScaleMeshSize(Scale Factor=&VectorFromCurve(
     VectorCurve=DataInterface() {propertyValues {
       XCurve {keys [{interpMode=RCIM_Cubic, tangentMode=RCTM_User, time=0, value=0.25}, {interpMode=RCIM_Cubic, tangentMode=RCTM_User, time=1, value=1}]}
       YCurve {keys [{interpMode=RCIM_Cubic, tangentMode=RCTM_User, time=0, value=0.25}, {interpMode=RCIM_Cubic, tangentMode=RCTM_User, time=1, value=1}]}
       ZCurve {keys [{arriveTangent=0.9, interpMode=RCIM_Cubic, time=0, value=0.5}, {interpMode=RCIM_Cubic, leaveTangent=0.9, time=1, value=2}]}
      }}
     Scale Curve=&MaskVectorBySpawnGroup(Masked Vector=<0.6 0.6 1>, Passthrough Vector=<1.5 1.5 1>, Spawn Group Mask=1)
    ))
   &DynamicMaterialParameters(
    Index 0 Param 1=&RandomRangeFloat(Minimum=0, Maximum=1)
    Index 0 Param 2=2
    Index 0 Param 3=1
    Index 1 Param 1=&FloatFromCurve(
     FloatCurve=DataInterface() {propertyValues {Curve {keys [
         {arriveTangent=-1.3693, interpMode=RCIM_Cubic, leaveTangent=-1.3693, tangentMode=RCTM_User, time=0, value=1}
         {interpMode=RCIM_Cubic, tangentMode=RCTM_User, time=1, value=0.5}
        ]}}}
     Scale Curve=20
    )
    Index 1 Param 2=&FloatFromCurve(
     FloatCurve=DataInterface() {propertyValues {Curve {keys [
         {arriveTangent=-1.4344, interpMode=RCIM_Cubic, leaveTangent=-1.4344, tangentMode=RCTM_User, time=0, value=1}
         {interpMode=RCIM_Cubic, tangentMode=RCTM_User, time=1, value=0.5}
        ]}}}
     Scale Curve=25
    )
    Index 1 Param 3=500
    Module.Param0WriteEnabled=true
    Module.Param0WriteEnabled1=true
    Module.Param1WriteEnabled=true
    Module.Param1WriteEnabled1=true
    Module.Param2WriteEnabled=true
    Module.Param2WriteEnabled1=true
    Module.Param3WriteEnabled=true
    Module.Param3WriteEnabled1=true
    Write Parameter Index 1=true
   ) {module=DynamicMaterialParameters001}
  ]
  rendererClasses [&NiagaraMeshRendererProperties]
  renderers [{rendererClass=&NiagaraMeshRendererProperties, rendererIndex=0}]
 }
 Streaks001 {
  emitterUpdateScript [
   &EmitterState(
    Inactive Response=::NewEnumerator1
    Life Cycle Mode=::NewEnumerator1
    Loop Behavior=::NewEnumerator1
    Loop Duration=$System.ChargeDuration
    UseLoopDelay=false
   )
   &SpawnRate(
    SpawnRate=&FloatFromCurve(FloatCurve=DataInterface() {propertyValues {Curve {keys [<at 0 1>,<at 1 0>]}}}, CurveIndex=$System.NormalizedLoopAge, Scale Curve=20)
    Use Spawn Probability=false
   )
  ]
  particleSpawnScript [
   &InitializeParticle(
    Color=<rgba 0 9.1318 20 1>
    Lifetime=0.5
    Sprite Size Max=<1.5 50>
    Sprite Size Min=<1 35>
    Sprite Size Mode=::NewEnumerator2
    UsePositionOffset=false
    Write Material Random=true
   )
   /SetVariables(Particles.Color=&LerpLinearColors(StartColor=$Particles.Color, EndColor=<rgba 20 7.9486 2.1875 1>, LerpFactor=$User.Blockbreak)) {enabled=false, isSetParameters=true}
   &NMS_Melee_SetColor(
    Block Break:NiagaraFloat=$User.Blockbreak
    Block Break Color:LinearColor=<rgba 20 7.9486 2.1875 1>
    Custom Color:LinearColor=$User.Color
    Custom Color Intensity:NiagaraFloat=$User.CustomColorIntensity
    Custom Color Multiplier=20
    Non Block Break Color:LinearColor=<rgba 0 9.1318 20 1>
    UseCustomColor:NiagaraFloat=$User.UseCustomColor
   )
   &AddVelocity(Module.Use Falloff Distance=false, Override Seed=false, Velocity Mode=::NewEnumerator1, Velocity Speed=0.1)
  ]
  particleUpdateScript [
   &ParticleState()
   &FN_Linear2PointSocketFlow(AnimationTime:NiagaraFloat=$System.DurationCurve, SkeletalMesh:NiagaraDataInterfaceSkeletalMesh=$User.SkeletalMesh)
   &SolveForcesAndVelocity(Clamp Velocity=false, Limit Acceleration=false)
   &ScaleSpriteSize(
    Scale Factor=&Vector2DFromCurve(Vector2Curve=DataInterface() {propertyValues {XCurve {keys [<at 0 1>,<at 1 0>]}, YCurve {keys [<at 0 1>,<at 1 0.2>]}}}, ScaleCurve=<1 1>)
    Scale Sprite Size Mode=::NewEnumerator2
   )
   &ScaleColor(
    Scale Alpha=&FloatFromCurve(FloatCurve=DataInterface() {propertyValues {Curve {keys [<at 0 1>,<at 0.8 0>]}}}, CurveIndex=$System.NormalizedLoopAge)
    ScaleA=true
    ScaleRGB=true
   )
  ]
  rendererClasses [&NiagaraSpriteRendererProperties]
  renderers [{rendererClass=&NiagaraSpriteRendererProperties}]
 }
 Tip_Flare {
  emitterUpdateScript [
   &EmitterState(
    Delay First Loop Only=true
    Inactive Response=::NewEnumerator2
    Life Cycle Mode=::NewEnumerator1
    Loop Delay=$System.ChargeDuration
    Loop Duration=1
    UseLoopDelay=true
   )
   &SpawnRate(SpawnRate=5, Use Spawn Probability=false)
  ]
  particleSpawnScript [
   &InitializeParticle(
    Color=<rgba 0.3476 2.3473 4 0.3>
    Lifetime=0.5
    Sprite Size Mode=::NewEnumerator3
    Uniform Sprite Size=110
    UsePositionOffset=false
    Write Material Random=true
   )
   /SetVariables(Particles.Color=&LerpLinearColors(StartColor=<rgba 0.2269 1.0576 4 1>, EndColor=<rgba 4 1.9919 1.1012 1>, LerpFactor=$User.Blockbreak)) {enabled=false, isSetParameters=true, module=SetVariables_#3}
   &NMS_Melee_SetColor(
    Block Break:NiagaraFloat=$User.Blockbreak
    Block Break Color:LinearColor=<rgba 4 1.9919 1.1012 1>
    Custom Color:LinearColor=$User.Color
    Custom Color Intensity:NiagaraFloat=$User.CustomColorIntensity
    Custom Color Multiplier=4
    Non Block Break Color:LinearColor=<rgba 0.2269 1.0576 4 1>
    UseCustomColor:NiagaraFloat=$User.UseCustomColor
   )
   &BoxLocation(Box Offset=<0 0 72.1851>, Box Size=<0 0 0>, MaskSpawnGroup=false, Module.Override Seed=false, Rotate Box=false, Surface Only=false)
   /SetVariables(
    Particles.GlowColor=&MultiplyLinearColor(Linear Color=$Particles.Color, Scale Factor=<1 1 1 0.1>)
    Particles.GlowPosition:NiagaraPosition=&Add_Vector(A=Linked() {linkedVariable {name=Particles.Position, type=NiagaraPosition}}, B=<0 0 -10>)
    Particles.GlowScale:Vector2f=&Multiply_Vector2D_ByFloat(Float=&RandomRangeFloat(Minimum=0.8, Maximum=1.1), Vector2D=<13 130>)
    Particles.SpriteAlignment=<0 0 1>
   ) {isSetParameters=true}
  ]
  particleUpdateScript [
   &ParticleState()
   &FN_Linear2PointSocketFlow(AnimationTime=1, SkeletalMesh:NiagaraDataInterfaceSkeletalMesh=$User.SkeletalMesh) {enabled=false}
   &ScaleSpriteSize()
   &CameraOffset(Camera Offset Amount=35)
   &DynamicMaterialParameters(
    Index 0 Param 1=0.2
    Index 0 Param 2=1
    Index 0 Param 3=1
    Index 0 Param 4=1
    Module.Param0WriteEnabled=true
    Module.Param1WriteEnabled=true
    Module.Param2WriteEnabled=true
    Module.Param3WriteEnabled=true
   )
   &ScaleColor(
    Scale Alpha=&FloatFromCurve(FloatCurve=DataInterface() {propertyValues {Curve {keys [
         {interpMode=RCIM_Cubic, tangentMode=RCTM_User, time=0, value=0}
         {interpMode=RCIM_Cubic, tangentMode=RCTM_User, time=0.1, value=1}
         <at 0.9 1 RCIM_Cubic>
         <at 0.9 1 RCIM_Cubic>
         <at 1 0 RCIM_Cubic>
        ]}}})
    ScaleA=true
    ScaleRGB=true
   )
  ]
  rendererClasses [&NiagaraSpriteRendererProperties]
  renderers [
   {rendererClass=&NiagaraSpriteRendererProperties}
   {rendererClass=&NiagaraSpriteRendererProperties, rendererIndex=1}
  ]
 }
}
summary {
 systemName=NS_Melee_ChargeUp_BasicSword
 emitters [
  {emitterName=Impact_Flare_and_Corona, rendererClasses [&NiagaraMeshRendererProperties]}
  {emitterName=Animorphic, rendererClasses [&NiagaraSpriteRendererProperties]}
  {emitterName=Flares, rendererClasses [&NiagaraSpriteRendererProperties]}
  {emitterName=Streaks001, rendererClasses [&NiagaraSpriteRendererProperties]}
  {emitterName=Charge_Flash, rendererClasses [&NiagaraSpriteRendererProperties]}
  {emitterName=Animorphic001, rendererClasses [&NiagaraSpriteRendererProperties]}
  {emitterName=Mesh, rendererClasses [&NiagaraMeshRendererProperties]}
  {emitterName=Spirals, rendererClasses [&NiagaraMeshRendererProperties]}
  {emitterName=Directional_sparks, simTarget=GPUComputeSim, rendererClasses [&NiagaraSpriteRendererProperties]}
  {emitterName=Smoke_Wide_Large, simTarget=GPUComputeSim, rendererClasses [&NiagaraSpriteRendererProperties]}
  {emitterName=Curled_Sparks, simTarget=GPUComputeSim, rendererClasses [&NiagaraSpriteRendererProperties]}
  {emitterName=Tip_Flare, rendererClasses [&NiagaraSpriteRendererProperties]}
 ]
 userVariables [
  User.Blockbreak {type=NiagaraFloat, defaultValue=0, description=""}
  User.Color {type=LinearColor, description="", defaultValue {struct=LinearColor, value=<rgba 0.2287 1 0 1>}}
  User.CustomColorIntensity {type=NiagaraFloat, defaultValue=1, description=""}
  User.Duration {type=NiagaraFloat, defaultValue=1, description=""}
  User.SkeletalMesh {
   type=NiagaraDataInterfaceSkeletalMesh
   defaultValue {
    struct=NiagaraExt_VariableValue_DataInterface
    value {dataInterface=&self:NiagaraDataInterfaceSkeletalMesh_1}
   }
  }
  User.UseCustomColor {type=NiagaraFloat, defaultValue=0, description=""}
 ]
}
